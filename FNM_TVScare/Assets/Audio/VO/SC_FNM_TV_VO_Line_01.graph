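# Sound cue that plays the first TV voice-over line, apparently for the FNM TV scare.
# A single wave player feeds a fixed delay of about one second, so the line starts shortly after the cue is triggered.
# The cue uses the TV voice-over attenuation asset, boosts volume and routes to the SFX sound class.
# A concurrency override limits simultaneous instances and sets a retrigger time.
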SoundCue /FNM_TVScare/Assets/Audio/VO/SC_FNM_TV_VO_Line_01
root=Delay_0
nodes [
 {id=Delay_0, children [WavePlayer_0], properties {delayMax=1.1, delayMin=1.1}}
 {id=WavePlayer_0, properties {soundWaveAssetPtr=&/FNM_TV_VO_Line01}}
]
properties {
 attenuationSettings=&/AS_FNM_TV_VO
 bOverrideConcurrency=true
 maxDistance=2700
 soundClassObject=&SFX
 volumeMultiplier=2.1
 concurrencyOverrides {maxCount=2, retriggerTime=2}
}
